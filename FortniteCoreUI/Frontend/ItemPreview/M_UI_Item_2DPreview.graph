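# UI material that shows an item's 2D preview texture with animated reveal and hide, using alpha composite blending.
# The texture scales to fit, with slight zoom driven by what appear to be the AnimIn and AnimOut parameters.
# Eased radial gradients and randomly offset, rounded column stripes build the transition masks.
# An optional FadeShadow switch clamps alpha with a linear gradient combined with an adjustable, blurred SDF circle.
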Material /FortniteCoreUI/Frontend/ItemPreview/M_UI_Item_2DPreview
graph {
 properties {
  bAutomaticallySetUsageInEditor=false
  bUsedWithStaticMesh=false
  blendMode=BLEND_AlphaComposite
  materialDomain=MD_UI
  usedShadingModels=Default Lit
 }
 nodes [
  Frac(None=Time())
  n1 = TextureObjectParameter() {parameterName=ItemTexture, texture=/Game/Athena/UI/Frontend/Spatial/Art/T_Ui_Empty}
  n2 = MaterialFunctionCall(UV=NamedRerouteUsage(), TextureSize=TextureProperty(Tex=NamedRerouteDeclaration(None=@n1) {name=Texture, nodeColor=<rgba 0 0 0 1>})) {materialFunction=&MF_UI_UVScaling}
  n3 = LinearInterpolate(A=LinearInterpolate(Alpha=NamedRerouteUsage()) {constA=0.9}, Alpha=NamedRerouteUsage()) {constB=0.9}
  n4 = TextureSample(
   UVs=NamedRerouteDeclaration(None=MaterialFunctionCall(UVs=@n2.Scaled to Fit, Texture Scale=@n3) {materialFunction=&ScaleUVsByCenter}.UVs) {name=Texture UV, nodeColor=<rgba 0 0 0 1>}
   Tex=NamedRerouteUsage()
  ) {samplerSource=SSM_Clamp_WorldGroupSettings}
  Multiply(A=NamedRerouteDeclaration(None=@n4.RGB) {name=texture color, nodeColor=<rgba 1 0.6563 0 1>}, B=NamedRerouteUsage())
  n5 = NamedRerouteUsage()
  n6 = Constant2Vector() {g=0.5}
  n7 = NamedRerouteUsage()
  n8 = MaterialFunctionCall(UV=MaterialFunctionCall(UVs=Add(A=@n5, B=@n6), Texture Scale=LinearInterpolate(Alpha=@n7) {constA=2, constB=0}) {materialFunction=&ScaleUVsByCenter}.UVs) {materialFunction=&MF_UI_RadialGradient}
  n9 = NamedRerouteUsage()
  NamedRerouteDeclaration(None=LinearInterpolate(
    B=Saturate(None=@n8.Smooth)
    Alpha=MaterialFunctionCall("Value (0-1)"=@n9) {materialFunction=&MF_UI_EaseCurves}.EaseOutQuint
   ) {constA=1}) {name=Unveil Mask, nodeColor=<rgba 0 0 0 1>}
  NamedRerouteDeclaration(None=MaterialFunctionCall!(UVs=NamedRerouteUsage(), Alpha=@n4.A) {materialFunction=/FortUILibrary/MaterialFunctions/Basics/MF_UI_ClipEdgePixel}.Result) {name=texture alpha, nodeColor=<rgba 0 1 0.9766 1>}
  n10 = Constant()
  n11 = Constant() {r=0.2}
  n12 = Constant() {r=0.3}
  n13 = Constant2Vector() {g=1}
  n14 = VectorParameter() {parameterName=CircleFadeScale, defaultValue=<rgba 1 1 0 0>}
  n15 = VectorParameter() {parameterName=CircleFadeOffset}
  n16 = MaterialFunctionCall!(
   Input UVs=MaterialFunctionCall(Texture Scale=AppendVector(A=@n14.RGB, B=@n14.RGB)) {materialFunction=&ScaleUVsByCenter}.UVs
   Translate=AppendVector(A=@n15.RGB, B=@n15.RGB)
  ) {materialFunction=/FortUILibrary/MaterialFunctions/Basics/MF_UI_Translate}
  n17 = ScalarParameter() {defaultValue=0.5, parameterName=CircleFadeRadius}
  n18 = MaterialFunctionCall("Value (0-1)"=MaterialFunctionCall!(
    DistanceField=MaterialFunctionCall!(UVs=@n16.UVs raw, Radius=@n17) {materialFunction=/FortUILibrary/MaterialFunctions/Shapes/SDFs/MF_SDF_Circle}.SDF
    GlowMin=ScalarParameter() {parameterName=BlurMin}
    GlowMax=ScalarParameter() {defaultValue=1, parameterName=BlurMax}
   ) {materialFunction=/FortUILibrary/MaterialFunctions/Shapes/Modifiers/MF_UI_SDFGlowOrAA}.Result) {materialFunction=&MF_UI_EaseCurves}
  NamedRerouteDeclaration(None=Max(
    A=MaterialFunctionCall(GradientStart=@n10, GradientEnd=@n11, RotationAngle=@n12, GradientOffset=@n13) {materialFunction=&MF_UI_LinearGradient}.Result
    B=Saturate(None=@n18.EaseInOutSine)
   )) {name=shadow fade, nodeColor=<rgba 0.4297 1 0 1>}
  NamedRerouteDeclaration(None=MaterialFunctionCall() {materialFunction=&GetUserInterfaceUV}.Normalized UV) {name=UV, nodeColor=<rgba 0 0 0 1>}
  NamedRerouteDeclaration(None=ScalarParameter() {parameterName=AnimOut, sliderMax=1}) {name=Anim Out, nodeColor=<rgba 0 0 0 1>}
  n19 = NamedRerouteUsage()
  n20 = Constant2Vector() {g=-0.5}
  n21 = NamedRerouteUsage()
  n22 = MaterialFunctionCall(UV=MaterialFunctionCall(UVs=Add(A=@n19, B=@n20), Texture Scale=LinearInterpolate(Alpha=@n21) {constA=0, constB=2}) {materialFunction=&ScaleUVsByCenter}.UVs) {materialFunction=&MF_UI_RadialGradient}
  n23 = NamedRerouteUsage()
  NamedRerouteDeclaration(None=MaterialFunctionCall!(
    1=Saturate(None=@n22.Smooth)
    2=NamedRerouteUsage()
    3=Constant() {r=1}
    4
    A=MaterialFunctionCall("Value (0-1)"=@n23) {materialFunction=&MF_UI_EaseCurves}.EaseInOutSine
   ) {materialFunction=/Engine/Functions/Engine_MaterialFunctions02/Lerp_Multiple_Float}.Lerp 3 Inputs) {name=Reveal Mask, nodeColor=<rgba 0 0 0 1>}
  NamedRerouteDeclaration(None=ScalarParameter() {defaultValue=1, parameterName=AnimIn, sliderMax=1}) {name=Anim In, nodeColor=<rgba 0 0 0 1>}
  n24 = NamedRerouteUsage()
  n25 = NamedRerouteUsage()
  n26 = Multiply(
   A=NamedRerouteUsage()
   B=LinearInterpolate(
    A=MaterialFunctionCall("Value (0-1)"=@n25) {materialFunction=&MF_UI_EaseCurves}.EaseInOutSine
    Alpha=NamedRerouteUsage()
   ) {constB=0}
  )
  NamedRerouteDeclaration(None=StaticSwitchParameter(True=Min(A=@n24, B=@n26), False=@n26) {parameterName=FadeShadow}) {name=final alpha, nodeColor=<rgba 1 0 0.7031 1>}
  n27 = MaterialFunctionCall!(
   UV=NamedRerouteUsage()
   Tiling X=NamedRerouteDeclaration(None=ScalarParameter() {defaultValue=64, parameterName=Tiling Amount}) {name=TilingX, nodeColor=<rgba 0 0 0 1>}
   Tiling Y=Constant() {r=1}
  ) {materialFunction=/FortUILibrary/MaterialFunctions/Basics/MF_UI_Pattern}
  n28 = MaterialFunctionCall(Float2=NamedRerouteDeclaration(None=@n27.Pattern UVs) {name=Pattern UV, nodeColor=<rgba 0 0 0 1>}) {materialFunction=&BreakOutFloat2Components}
  n29 = NamedRerouteUsage()
  n30 = Multiply(A=@n28.G, B=@n29)
  n31 = Multiply(A=Subtract(A=@n29)) {constB=-0.5}
  n32 = Subtract(A=AppendVector(A=@n28.R, B=Add(A=@n30, B=@n31) {constB=-0.5})) {constB=0.5}
  n33 = Constant2Vector() {desc=Padding between pips, g=2, r=1.5}
  n34 = Multiply(A=NamedRerouteUsage()) {constB=2}
  n35 = Multiply(A=@n34) {constB=-1}
  n36 = MaterialFunctionCall!(UV=Floor(None=@n27.Raw UVs), Seed=Constant() {r=0.4}) {materialFunction=&MF_UI_Random}
  n37 = Frac(None=Add(A=NamedRerouteDeclaration(None=@n36.Result) {name=Random Columns, nodeColor=<rgba 0 0 0 1>}, B=NamedRerouteUsage()))
  n38 = Add(
   A=Add(A=Multiply(A=@n32, B=@n33)) {constB=0.5}
   B=AppendVector(A=Constant(), B=LinearInterpolate(A=@n35, B=@n34, Alpha=@n37) {constA=-96, constB=96})
  )
  n39 = AppendVector(A=Constant() {r=1}, B=NamedRerouteUsage())
  n40 = Constant() {r=1}
  NamedRerouteDeclaration(None=Saturate(None=MaterialFunctionCall(UV=@n38, Size=@n39, Corner Radius=@n40) {materialFunction=&MF_UI_CoreUI_RoundedMask}.Shape Mask)) {name=Stripes, nodeColor=<rgba 0 0 0 1>}
  NamedRerouteUsage()
 ]
}
