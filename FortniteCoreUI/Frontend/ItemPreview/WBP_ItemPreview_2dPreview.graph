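# Activatable frontend item-preview widget that shows a downloaded 2D image as a stand-in while an item's 3D asset downloads.
# A URL change loads a FortLazyImage after a CVar-set delay, then fades it in through material parameter animations.
# When 3D finishes it fades out or clears the image, while load errors swap in an error material.
# A separate throbber shows an async-loaded item-category icon, and delegates report load start and finish.
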WidgetBlueprint /FortniteCoreUI/Frontend/ItemPreview/WBP_ItemPreview_2dPreview
animations [
 AnimIn {
  binding_count=1
  length_seconds=0.25
  object=&self:AnimIn
  sequence {bindings [LazyImage_2dPreview {
     id=#1
     possessed_class=FortLazyImage
     tracks [{
       type=MovieSceneWidgetMaterialTrack
       sections [{
         type=MovieSceneParameterSection
         end_frame=5
         start_frame=0
         channels [
          AnimIn {type=MovieSceneScriptingFloatChannel, keys [[0,0],[4.9997,1]]}
          AnimOut {type=MovieSceneScriptingFloatChannel, const [0,1]}
         ]
        }]
      }]
    }]}
 }
 AnimOut {
  binding_count=1
  length_seconds=0.25
  object=&self:AnimOut
  sequence {bindings [LazyImage_2dPreview {
     id=#1
     possessed_class=FortLazyImage
     tracks [{
       type=MovieSceneWidgetMaterialTrack
       sections [{
         type=MovieSceneParameterSection
         end_frame=5
         start_frame=0
         channels [
          AnimOut {type=MovieSceneScriptingFloatChannel, thinned_from=3, keys [[0,0],[5,1]]}
          AnimIn {type=MovieSceneScriptingFloatChannel, const [1,1]}
         ]
        }]
      }]
    }]}
 }
]
graphs [
 CleanupTexture {nodes [
   CleanupTexture() {is=title K2Node_FunctionEntry}
   > n1 = Branch(Condition=Get IsShowingDefaultMaterial() -> (IsShowingDefaultMaterial:bool))
   n2 = GetDynamicMaterial(execute=[@n1.then, @n4.then], self=Get LazyImage_2dPreview() -> (LazyImage_2dPreview:Fort Lazy Image Object Reference))
   > SetTextureParameterValue(self=@n2, ParameterName=ItemTexture, Value=/Game/Athena/UI/Frontend/Spatial/Art/T_Ui_Empty)
   > DebugTextWithUrl!(execute, self, TextBeforeUrl:String="Cleanup texture ") -> (then)
   n3 = Get LazyImage_2dPreview()
   SetBrushFromMaterial(execute=@n1.else, self=@n3, Material=Get 2dPreviewMaterial() -> (2dPreviewMaterial:Material Interface Object Reference))
   > SetMaterialTextureParamName!(execute, self:Common Lazy Image Object Reference=@n3, TextureParamName:Name=ItemTexture) -> (then)
   > n4 = Set IsShowingDefaultMaterial(IsShowingDefaultMaterial:bool=true) -> (Output_Get:bool)
  ]}
 DebugTextWithUrl {nodes [
   n1 = DebugTextWithUrl!() -> (then, TextBeforeUrl:String) {is=title K2Node_FunctionEntry}
   > Branch(Condition=Get Enable2dPreviewDebug() -> (Enable2dPreviewDebug:bool))
   > PrintText!(
    execute
    InText=Format Text!(Format="INVTEXT(\"{text} {url}\")", text:String=@n1, url:String=Get CachedUrl() -> (CachedUrl:String)) -> (Result)
    bPrintToScreen=false
    bPrintToLog
    TextColor
    Duration
    Key
   ) -> (then)
  ]}
 EventGraph {nodes [
   "On Loading Errored (LazyImage_2dPreview)"!() -> (OutputDelegate, then, ErrorCode:int=0, ModuleId:int=0, ErrorString:String, ErrorMessage:Text) {is=title K2Node_ComponentBoundEvent}
   > Call On 2d Finished Loading!(execute:Exec, self:WBP Item Preview 2d Preview Object Reference, Success:bool, ImageUrl:String=Get CachedUrl()) -> (then:Exec) {is=title K2Node_CallDelegate}
   n1 = Get CachedUrl()
   > n2 = Branch(Condition=NOT Boolean(A=IsEmpty(InString=@n1)))
   > n3 = Branch(Condition=NOT Boolean(A=Get Is3dDownloaded() -> (Is3dDownloaded:bool)))
   > n4 = ShowErrorPresentation()
   SetLoadingThrobberVisibility!(execute=[@n5.then, @n4.then], self, IsVisible:bool=false) -> (then)
   n5 = CleanupTexture(execute=[@n3.else, @n2.else]) {is=title K2Node_CallFunction}
   OnImageLoaded!() -> (OutputDelegate, then, Texture:Texture 2D Object Reference) {is=title K2Node_CustomEvent}
   > Branch(Condition=Get Is2dPreviewEnabled() -> (Is2dPreviewEnabled:bool))
   n6 = Get CachedUrl()
   > n7 = Branch(Condition=NOT Boolean(A=IsEmpty(InString=@n6)))
   > Branch(Condition=NOT Boolean(A=Get Is3dDownloaded()))
   > Call On 2d Finished Loading(Success=true, ImageUrl=Get CachedUrl())
   > SetLoadingThrobberVisibility(IsVisible=false)
   > StopAnimation(InAnimation=Get AnimIn() -> (AnimIn:Widget Animation Object Reference))
   > StopAnimation(InAnimation=Get AnimOut() -> (AnimOut:Widget Animation Object Reference))
   > PlayAnimationForward(InAnimation=Get AnimIn())
   > DebugTextWithUrl(TextBeforeUrl=Image loaded for) {is=title K2Node_CallFunction}
   > RestoreCachedScale()
   SetLoadingThrobberVisibility(execute=@n7.else, IsVisible=false)
   > CleanupTexture()
   On3dItemDownloaded() {is=title K2Node_CustomEvent}
   > Branch(Condition=Get Is2dPreviewEnabled())
   n8 = Get CachedUrl()
   > n9 = Branch(Condition=NOT Boolean(A=IsEmpty(InString=@n8)))
   > Branch(Condition=NOT Boolean(A=Get Is3dDownloaded()))
   > Set Is3dDownloaded(Is3dDownloaded:bool=true) -> (Output_Get:bool)
   > SetLoadingThrobberVisibility(IsVisible=false)
   > DebugTextWithUrl(TextBeforeUrl=3D asset download complete for)
   > n10 = Branch(Condition=Get Has2dLoadingError() -> (Has2dLoadingError:bool))
   > CleanupTexture()
   > PlayAnimationForward(InAnimation=Get AnimOut(), PlaybackSpeed=2)
   > Set Has2dLoadingError(Has2dLoadingError:bool=false) -> (Output_Get:bool)
   n11 = Branch(execute=@n10.else, Condition=IsLoading!(self:Common Lazy Image Object Reference=Get LazyImage_2dPreview()) -> (ReturnValue:bool))
   > CleanupTexture()
   > PlayAnimationForward(InAnimation=Get AnimOut())
   > Call On 2d Finished Loading(ImageUrl=Get CachedUrl())
   n12 = Branch(execute=@n11.else, Condition=IsAnimationPlaying(InAnimation=Get AnimIn()))
   > PlayAnimationReverse(InAnimation=Get AnimIn())
   PlayAnimationForward(execute=@n12.else, InAnimation=Get AnimOut())
   SetLoadingThrobberVisibility(execute=@n9.else, IsVisible=false)
   > CleanupTexture()
   Toggle2dPreviewDebug() {is=title K2Node_CustomEvent}
   > Set Enable2dPreviewDebug(Enable2dPreviewDebug:bool=NOT Boolean(A=Get Enable2dPreviewDebug())) -> (Output_Get:bool)
   n13 = LoadLoadingThrobberTexture!() -> (OutputDelegate, then, TextureToLoad:Texture 2D Soft Object Reference) {is=title K2Node_CustomEvent}
   > n14 = Async Load Asset(Asset=@n13.TextureToLoad)
   > Set LastLoadingThrobberSoftPath(LastLoadingThrobberSoftPath:String=BreakSoftObjectPath(InSoftObjectPath=To Soft Object Path!(SoftObjectReference:Soft Object Reference=@n13.TextureToLoad) -> (ReturnValue:Soft Object Path Structure))) -> (Output_Get:String)
   n15 = Cast To Texture2D(execute=@n14.Completed, Object=@n14)
   > n16 = Set LoadingThrobberLoadedTexture(LoadingThrobberLoadedTexture:Texture 2D Object Reference=@n15) -> (Output_Get:Texture 2D Object Reference)
   n17 = Get Is3dDownloaded()
   n18 = IsEmpty(InString=Get CachedUrl())
   Branch(execute=[@n21.then, @n16.then], Condition=AND Boolean(A=NOT Boolean(A=@n17), B=NOT Boolean(A=@n18)))
   n19 = "Equal Exactly (String)"(
    A=Get LastLoadingThrobberSoftPath() -> (LastLoadingThrobberSoftPath:String)
    B=Get Object Path String(Object=Get LoadingThrobberLoadedTexture() -> (LoadingThrobberLoadedTexture:Texture 2D Object Reference))
   )
   n20 = NOT Boolean(A=Get HasLoadingThrobberError() -> (HasLoadingThrobberError:bool))
   > SetLoadingThrobberTexture!(
    execute
    self
    Texture:Texture 2D Object Reference=Select(
     Option 0:Texture 2D Object Reference=Get DefaultLoadingThrobberTexture() -> (DefaultLoadingThrobberTexture:Texture 2D Object Reference)
     Option 1:Texture 2D Object Reference=Get LoadingThrobberLoadedTexture()
     Index=AND Boolean(A=@n19, B=@n20)
    ) -> (ReturnValue:Texture 2D Object Reference)
   ) -> (then)
   > SetLoadingThrobberVisibility(IsVisible=true)
   Set HasLoadingThrobberError(execute=@n15.CastFailed, HasLoadingThrobberError:bool=true) -> (Output_Get:bool)
   > PrintString!(execute, InString=Failed to download loading throbber texture, bPrintToScreen, bPrintToLog, TextColor, Duration, Key) -> (then)
   > n21 = Set LoadingThrobberLoadedTexture()
   n22 = OnLoadingStarted() {is=title K2Node_CustomEvent}
   > Call On 2d Started Loading!(execute:Exec, self:WBP Item Preview 2d Preview Object Reference, ImageUrl:String=Get CachedUrl()) -> (then:Exec) {is=title K2Node_CallDelegate}
   Event On Activated()
   > n23 = Branch(Condition=Get Is2dPreviewEnabled())
   > n24 = GetConsoleVariableFloatValue!(execute, VariableName:String=Athena.Frontend.2dPreviewDownloadDelay) -> (then, ReturnValue:float)
   > Set UrlDownloadDelay(UrlDownloadDelay:float=@n24) -> (Output_Get:float)
   > StopAnimation(InAnimation=Get AnimIn())
   > StopAnimation(InAnimation=Get AnimOut())
   > Unbind all Events from On Loading Started!(execute:Exec, self:Fort Lazy Image Object Reference=Get LazyImage_2dPreview()) -> (then:Exec) {is=title K2Node_ClearDelegate}
   > Bind Event to On Loading Started(self:Fort Lazy Image Object Reference=Get LazyImage_2dPreview(), Delegate=@n22) {is=title K2Node_AddDelegate}
   Set Visibility(execute=@n23.else, Visibility=Collapsed, self=Get LazyImage_2dPreview())
   Event PreConstruct()
   > Set IsShowingDefaultMaterial(IsShowingDefaultMaterial=true)
  ]}
 On2dFinishedLoading {nodes [On2dFinishedLoading!() -> (then, Success:bool=~, ImageUrl:String) {is=title K2Node_FunctionEntry}]}
 On2dStartedLoading {nodes [On2dStartedLoading!() -> (then, ImageUrl:String) {is=title K2Node_FunctionEntry}]}
 RestoreCachedScale {nodes [
   RestoreCachedScale() {is=title K2Node_FunctionEntry}
   n1 = Get LazyImage_2dPreview()
   n2 = Break Widget Transform!(WidgetTransform:"Widget Transform Structure (by ref)"=Get RenderTransform!(self:Widget Object Reference=@n1) -> (RenderTransform:Widget Transform Structure)) -> (Translation:Vector 2D Structure, Scale:Vector 2D Structure, Shear:Vector 2D Structure, Angle:float=~) {is=title K2Node_BreakStruct}
   > Set RenderTransform!(
    execute
    RenderTransform:Widget Transform Structure=Make Widget Transform!(
     Translation:Vector 2D Structure=@n2
     Scale:Vector 2D Structure=Get CachedScale() -> (CachedScale:Vector 2D Structure)
     Shear:Vector 2D Structure=@n2.Shear
     Angle:float=@n2.Angle
    ) -> (WidgetTransform:Widget Transform Structure) {is=title K2Node_MakeStruct}
    self:Widget Object Reference=@n1
   ) -> (then, Output_Get:Widget Transform Structure)
  ]}
 SetErrorImageScale {nodes [
   SetErrorImageScale() {is=title K2Node_FunctionEntry}
   n1 = Get TargetScale() -> (TargetScale:Vector 2D Structure)
   n2 = Break Widget Transform(WidgetTransform=Get RenderTransform(self=Get LazyImage_2dPreview())) -> (Angle=~)
   > Branch(Condition="Not Equal (Vector2D)"!(A:Vector 2D Structure=@n1, B:Vector 2D Structure=@n2.Scale, ErrorTolerance:float) -> (ReturnValue:bool=~) {is=title K2Node_PromotableOperator})
   > Set CachedScale(CachedScale:Vector 2D Structure=@n2.Scale) -> (Output_Get:Vector 2D Structure)
   > Set RenderTransform(RenderTransform=Make Widget Transform(Translation=@n2, Scale=@n1, Shear=@n2.Shear, Angle=@n2.Angle), self=Get LazyImage_2dPreview())
  ]}
 SetLoadingThrobberTexture {nodes [
   SetLoadingThrobberTexture!() -> (then, Texture:Texture 2D Object Reference) {is=title K2Node_FunctionEntry}
   > n1 = GetDynamicMaterial(self=Get LoadingThrobber() -> (LoadingThrobber:Image Object Reference))
   > SetTextureParameterValue(self=@n1, ParameterName=ItemTexture, Value=Get Texture() -> (Texture:Texture 2D Object Reference))
  ]}
 SetLoadingThrobberVisibility {nodes [
   n1 = SetLoadingThrobberVisibility!() -> (then, IsVisible:bool=~) {is=title K2Node_FunctionEntry}
   > Set Visibility(
    Visibility=Select(Option 0:ESlateVisibility Enum=Collapsed, Option 1:ESlateVisibility Enum=SelfHitTestInvisible, Index=@n1) -> (ReturnValue:ESlateVisibility Enum)
    self=Get LoadingThrobber()
   )
  ]}
 ShowErrorPresentation {nodes [
   ShowErrorPresentation() {is=title K2Node_FunctionEntry}
   > SetBrushFromMaterial(self=Get LazyImage_2dPreview(), Material=Get ErrorMaterial() -> (ErrorMaterial:Material Interface Object Reference))
   > SetErrorImageScale() {is=title K2Node_CallFunction}
   > Set IsShowingDefaultMaterial(IsShowingDefaultMaterial=false)
   > Set Has2dLoadingError(Has2dLoadingError=true)
   > DebugTextWithUrl(TextBeforeUrl="2D loading error for ")
  ]}
 UpdateLoadingThrobberTexture {nodes [
   UpdateLoadingThrobberTexture!() -> (then, ItemVM:Fort Item VM Object Reference) {is=title K2Node_FunctionEntry}
   n1 = Get Is3dDownloaded()
   > n2 = Branch(Condition=AND Boolean!(A=NOT Boolean(A=@n1), B=Get Is2dPreviewEnabled(), C:bool=Get CanShowLoadingThrobber() -> (CanShowLoadingThrobber:bool)) -> (ReturnValue))
   > Set HasLoadingThrobberError(HasLoadingThrobberError=false)
   > n3 = Is Valid(InputObject=Get ItemVM() -> (ItemVM:Fort Item VM Object Reference))
   SetLoadingThrobberTexture(execute=@n3.Is Not Valid, Texture=Get DefaultLoadingThrobberTexture()) {is=title K2Node_CallFunction}
   n4 = Get ItemVM()
   LoadLoadingThrobberTexture!(
    execute=@n3.Is Valid
    self
    TextureToLoad:Texture 2D Soft Object Reference=Get LoadingIcon2dPreview!(self:Fort Locker Category Config Object Reference=GetItemCategory!(self:Fort Item VM Object Reference=@n4) -> (ReturnValue:Fort Locker Category Config Object Reference)) -> (LoadingIcon2dPreview:Texture 2D Soft Object Reference)
   ) -> (then) {is=title K2Node_CallFunction}
   SetLoadingThrobberVisibility(execute=@n2.else, IsVisible=false) {is=title K2Node_CallFunction}
  ]}
 VM_2dUrlChanged {nodes [
   VM_2dUrlChanged!() -> (then, Url:String) {is=title K2Node_FunctionEntry}
   > n1 = Branch(Condition=Get Is2dPreviewEnabled())
   > Set Is3dDownloaded(Is3dDownloaded=false)
   > Set Has2dLoadingError(Has2dLoadingError=false)
   > Set CachedUrl(CachedUrl:String=Get Url() -> (Url:String)) -> (Output_Get:String)
   > n2 = Branch(Condition=Get Is2dPreviewEnabled())
   n3 = Get Url()
   > n4 = Branch(Condition=NOT Boolean(A=IsEmpty(InString=@n3)))
   > Set CanShowLoadingThrobber(CanShowLoadingThrobber:bool=true) -> (Output_Get:bool)
   > Set Visibility(Visibility=SelfHitTestInvisible, self=Get LazyImage_2dPreview())
   > CleanupTexture()
   > SetBrushFromURLWithCallback!(
    execute
    self:Fort Lazy Image Object Reference=Get LazyImage_2dPreview()
    CMSImageURL:String=Get Url()
    bMatchSize:bool=false
    OnLoadedCallback:"Delegate (by ref)"=Create Event()
    Delay:float=Get UrlDownloadDelay() -> (UrlDownloadDelay:float)
    DownloadPriority:ECMSDownloadPriority Enum=Normal
   ) -> (then)
   > DebugTextWithUrl(TextBeforeUrl="URL set for ")
   Set CanShowLoadingThrobber(execute=@n4.else, CanShowLoadingThrobber=false)
   > Call On 2d Finished Loading(Success=true, ImageUrl=Get CachedUrl())
   > SetLoadingThrobberVisibility(IsVisible=false)
   > CleanupTexture()
   Set CanShowLoadingThrobber(execute=@n2.else, CanShowLoadingThrobber=false)
   > CleanupTexture()
   Set Visibility(execute=@n1.else, Visibility=Collapsed, self=Get LazyImage_2dPreview())
   > SetLoadingThrobberVisibility(IsVisible=false)
  ]}
]
variables [
 2dPreviewMaterial {type=MaterialInterface, default=&/MI_UI_Item_2DPreview}
 CachedScale {type=Vector2D, default=<1 1>}
 CachedUrl {type=string, default=""}
 CanShowLoadingThrobber {type=boolean}
 DefaultLoadingThrobberTexture {type=Texture2D}
 Enable2dPreviewDebug {type=boolean}
 ErrorMaterial {type=MaterialInterface, default=/FortUILibrary/UIKit/Throbber/MI_UIKit_Color_Throbber}
 Has2dLoadingError {type=boolean}
 HasLoadingThrobberError {type=boolean}
 Is2dPreviewEnabled {type=boolean, default=true}
 Is3dDownloaded {type=boolean}
 IsShowingDefaultMaterial {type=boolean, default=true}
 LastLoadingThrobberSoftPath {type=string, default=""}
 LoadingThrobberLoadedTexture {type=Texture2D, default=/FortUILibrary/Textures/Icons/T_UI_Icon_UIKit_Wraps_MSDF}
 On2dFinishedLoading {}
 On2dStartedLoading {}
 UrlDownloadDelay {type=number, default=2}
]
widgets {
 info {inheritedWidgetCount=0, parentClass=&CommonActivatableWidget, rootWidgetClass=&Overlay, widgetCount=3}
 widgets [
  Overlay_0 {widgetClassPath=&Overlay, widget=Overlay_0}
  LoadingThrobber {
   widgetClassPath=&Image
   bIsVariable=true
   parent=Overlay_0
   widget=LoadingThrobber
   properties {brush {
     bIsDynamicallyLoaded=false
     drawAs=Image
     imageType=FullColor
     mirroring=NoMirror
     resourceName=None
     resourceObject=&/MI_UI_Throbber_LoadingAsset
     tiling=NoTile
     imageSize=<256 256>
     margin=<ltrb 0 0 0 0>
     outlineSettings {
      bUseBrushTransparency=false
      roundingType=HalfHeightRadius
      width=0
      color {colorUseRule=UseColor_Specified, specifiedColor=<rgba 0 0 0 0>}
      cornerRadii=<0 0 0 0>
     }
     tintColor {colorUseRule=UseColor_Specified, specifiedColor=<rgba 1 1 1 1>}
     uVRegion {bIsValid=false, max=<0 0>, min=<0 0>}
    }}
   slot {class=OverlaySlot, properties {horizontalAlignment=HAlign_Center, verticalAlignment=VAlign_Center}}
  }
  LazyImage_2dPreview {
   widgetClassPath=&FortLazyImage
   bIsVariable=true
   parent=Overlay_0
   widget=LazyImage_2dPreview
   properties {
    bShowLoading=true
    materialTextureParamName=ItemTexture
    brush {
     bIsDynamicallyLoaded=false
     drawAs=Image
     imageType=FullColor
     mirroring=NoMirror
     resourceName=None
     resourceObject=&MI_UI_Item_2DPreview
     tiling=NoTile
     imageSize=<32 32>
     margin=<ltrb 0 0 0 0>
     outlineSettings {
      bUseBrushTransparency=false
      roundingType=HalfHeightRadius
      width=0
      color {colorUseRule=UseColor_Specified, specifiedColor=<rgba 0 0 0 0>}
      cornerRadii=<0 0 0 0>
     }
     tintColor {colorUseRule=UseColor_Specified, specifiedColor=<rgba 1 1 1 1>}
     uVRegion {bIsValid=false, max=<0 0>, min=<0 0>}
    }
    loadingBackgroundBrush {
     bIsDynamicallyLoaded=false
     drawAs=Image
     imageType=FullColor
     mirroring=NoMirror
     resourceName=None
     resourceObject=/Game/Global/Textures/Default/Blanks/T_Blank_White_ZeroAlpha_sRGBColor
     tiling=NoTile
     imageSize=<1 1>
     margin=<ltrb 0 0 0 0>
     outlineSettings {
      bUseBrushTransparency=false
      roundingType=HalfHeightRadius
      width=0
      color {colorUseRule=UseColor_Specified, specifiedColor=<rgba 0 0 0 0>}
      cornerRadii=<0 0 0 0>
     }
     tintColor {colorUseRule=UseColor_Specified, specifiedColor=<rgba 1 1 1 1>}
     uVRegion {bIsValid=false, max=<0 0>, min=<0 0>}
    }
    loadingThrobberBrush {
     bIsDynamicallyLoaded=false
     drawAs=Image
     imageType=FullColor
     mirroring=NoMirror
     resourceName=None
     resourceObject=&T_Blank_White_ZeroAlpha_sRGBColor
     tiling=NoTile
     imageSize=<1 1>
     margin=<ltrb 0 0 0 0>
     outlineSettings {
      bUseBrushTransparency=false
      roundingType=HalfHeightRadius
      width=0
      color {colorUseRule=UseColor_Specified, specifiedColor=<rgba 0 0 0 0>}
      cornerRadii=<0 0 0 0>
     }
     tintColor {colorUseRule=UseColor_Specified, specifiedColor=<rgba 1 1 1 1>}
     uVRegion {bIsValid=false, max=<0 0>, min=<0 0>}
    }
   }
   slot {class=OverlaySlot, properties {horizontalAlignment=HAlign_Fill, verticalAlignment=VAlign_Fill}}
  }
 ]
}
