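# WBP_ItemPreview_Info is a frontend item preview status widget that shows download progress, errors or an "X of Y" variant counter.
# UpdateLoading stores the 3D load state and switches between loading, error and empty displays, each setting icon, text and backplate.
# UpdateVariantsText shows the counter only when no loading or error state is active.
# IsCompact reduces loading to a throbber without text or backplate; the 2D state only selects identical loading messages.
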WidgetBlueprint /FortniteCoreUI/Frontend/ItemPreview/WBP_ItemPreview_Info
graphs [
 Display_Empty {nodes [
   Display_Empty() {is=title K2Node_FunctionEntry}
   > SetVisibility(self=Get SizeBox_Icon() -> (SizeBox_Icon:Size Box Object Reference), InVisibility=Collapsed)
   > Set Text(self=Get Text!() -> (Text:Common Text Object Reference))
   > SetBackplateVisible!(execute, self, IsVisible:bool=false) -> (then)
  ]}
 Display_Error {nodes [
   Display_Error!() -> (then, Message:Text) {is=title K2Node_FunctionEntry}
   > SetVisibility(self=Get SizeBox_Icon(), InVisibility=HitTestInvisible)
   > SetActiveWidget(
    self=Get Switcher_Icon() -> (Switcher_Icon:Common Visibility Switcher Object Reference)
    Widget=Get Error_Indicator() -> (Error_Indicator:WBP UIKit Status Indicator Object Reference)
   )
   > Set Text(Text=Get Message() -> (Message:Text), self=Get Text())
   > SetBackplateVisible(IsVisible=true)
  ]}
 Display_Loading {nodes [
   Display_Loading!() -> (then, Message:Text) {is=title K2Node_FunctionEntry}
   > SetVisibility(self=Get SizeBox_Icon(), InVisibility=HitTestInvisible)
   > SetActiveWidget(self=Get Switcher_Icon(), Widget=Get Loading_Throbber() -> (Loading_Throbber:WBP UIKit Throbber Object Reference))
   n1 = Get IsCompact() -> (IsCompact:bool)
   > Set Text(Text=SelectText(B=Get Message(), bPickA=@n1), self=Get Text())
   > SetBackplateVisible(IsVisible=NOT Boolean(A=@n1))
  ]}
 Display_Variants {nodes [
   Display_Variants!() -> (then, Message:Text) {is=title K2Node_FunctionEntry}
   > SetVisibility(self=Get SizeBox_Icon(), InVisibility=Collapsed)
   > Set Text(Text=Get Message(), self=Get Text())
   > SetBackplateVisible(IsVisible=true)
  ]}
 EventGraph {nodes [
   n1 = Event PreConstruct()
   > n2 = Sequence()
   > Display_Empty() {is=title K2Node_CallFunction}
   > Set PreviewLoadState(PreviewLoadState:E_ItemPreview_LoadState Enum=NewEnumerator3) -> (Output_Get:E_ItemPreview_LoadState Enum=NewEnumerator3)
   Branch(execute=@n2.then_1, Condition=@n1.IsDesignTime)
   > UpdateLoading!(
    execute
    self
    LoadState2D:E_ItemPreview_LoadState Enum=NewEnumerator0
    LoadState3D:E_ItemPreview_LoadState Enum=NewEnumerator0
    ErrorMessage:Text
   ) -> (then)
  ]}
 HasLoadingInfoToShow {nodes [
   HasLoadingInfoToShow() {is=title K2Node_FunctionEntry}
   > Return Node!(
    execute
    HasInfo:bool=Select!(
     NewEnumerator3:bool=false
     NewEnumerator0:bool=true
     NewEnumerator2:bool=false
     NewEnumerator1:bool=true
     Index:E_ItemPreview_LoadState Enum=Get PreviewLoadState() -> (PreviewLoadState:E_ItemPreview_LoadState Enum=NewEnumerator3)
    ) -> (ReturnValue:bool)
   )
  ]}
 SetBackplateVisible {nodes [
   n1 = SetBackplateVisible!() -> (then, IsVisible:bool=~) {is=title K2Node_FunctionEntry}
   > SetVisibility(
    self=Get Backplate() -> (Backplate:WBP UIKit Backplate Object Reference)
    InVisibility=Select!(Option 0:ESlateVisibility Enum=Collapsed, Option 1:ESlateVisibility Enum=HitTestInvisible, Index:bool=@n1) -> (ReturnValue:ESlateVisibility Enum)
   )
   > SetPadding!(
    execute
    self:Horizontal Box Slot Object Reference=SlotAsHorizontalBoxSlot!(Widget:Widget Object Reference=Get SizeBox_Icon()) -> (ReturnValue:Horizontal Box Slot Object Reference)
    InPadding:Margin Structure=Make Margin(Left=SelectFloat(A=16, B=0, bPickA=@n1), Top=0, Right=0, Bottom=0)
   ) -> (then)
  ]}
 UpdateLoading {nodes [
   n1 = UpdateLoading!() -> (then, LoadState2D:E_ItemPreview_LoadState Enum=NewEnumerator0, LoadState3D:E_ItemPreview_LoadState Enum=NewEnumerator0, ErrorMessage:Text) {is=title K2Node_FunctionEntry}
   > Set PreviewLoadState(PreviewLoadState=@n1.LoadState3D) -> (Output_Get=NewEnumerator3)
   > n2 = Switch on E_ItemPreview_LoadState!(execute, Selection:E_ItemPreview_LoadState Enum=Get LoadState3D() -> (LoadState3D:E_ItemPreview_LoadState Enum=NewEnumerator3)) -> (NewEnumerator3:Exec, NewEnumerator0:Exec, NewEnumerator2:Exec, NewEnumerator1:Exec) {is=title K2Node_SwitchEnum}
   n3 = Get ErrorMessage() -> (ErrorMessage:Text)
   n4 = TextIsEmpty(InText=@n3)
   Display_Error!(
    execute=@n2.NewEnumerator1
    self
    Message:Text=Select(
     Option 0:Text=MakeLiteralText(Value="LOCTEXT(\"Item could not be downloaded. Please try again later.\")")
     Option 1:Text=@n3
     Index=NOT Boolean(A=@n4)
    ) -> (ReturnValue:Text)
   ) -> (then) {is=title K2Node_CallFunction}
   Display_Empty(execute=[@n2.NewEnumerator2, @n2.NewEnumerator3])
   n5 = MakeLiteralText(Value="LOCTEXT(\"Downloading Item...\")")
   Display_Loading!(
    execute=@n2.NewEnumerator0
    self
    Message:Text=Select!(
     NewEnumerator3:Text=@n5
     NewEnumerator0:Text=@n5
     NewEnumerator2:Text=@n5
     NewEnumerator1:Text=@n5
     Index:E_ItemPreview_LoadState Enum=Get LoadState2D() -> (LoadState2D:E_ItemPreview_LoadState Enum=NewEnumerator3)
    ) -> (ReturnValue)
   ) -> (then) {is=title K2Node_CallFunction}
  ]}
 UpdateVariantsText {nodes [
   n1 = UpdateVariantsText!() -> (then, CurrentVariant:int, TotalVariants:int) {is=title K2Node_FunctionEntry}
   > n2 = Branch(Condition=HasLoadingInfoToShow!(self) -> (HasInfo:bool) {is=title K2Node_CallFunction})
   Display_Variants!(
    execute=@n2.else
    self
    Message:Text=Format Text!(Format="LOCTEXT(\"{0} of {1}\")", 0:int=int + int(A=@n1), 1:int=@n1.TotalVariants) -> (Result)
   ) -> (then) {is=title K2Node_CallFunction}
  ]}
]
variables [
 IsCompact {type=boolean, default=true}
 PreviewLoadState {type=E_ItemPreview_LoadState, default=Unset}
]
widgets {
 info {inheritedWidgetCount=0, parentClass=&UserWidget, rootWidgetClass=&Overlay, widgetCount=8}
 widgets [
  Overlay_LoadingModels {widgetClassPath=&Overlay, widget=Overlay_LoadingModels}
  Backplate {
   widgetClassPath=&WBP_UIKit_Backplate
   bIsVariable=true
   parent=Overlay_LoadingModels
   widget=Backplate
   properties {brightness=Dark}
   slot {class=OverlaySlot, properties {horizontalAlignment=HAlign_Fill, verticalAlignment=VAlign_Fill}}
  }
  HorizontalBox_2 {
   widgetClassPath=&HorizontalBox
   parent=Overlay_LoadingModels
   widget=HorizontalBox_2
   slot {class=OverlaySlot, properties {horizontalAlignment=HAlign_Fill, verticalAlignment=VAlign_Center}}
  }
  SizeBox_Icon {
   widgetClassPath=&SizeBox
   bIsVariable=true
   parent=HorizontalBox_2
   widget=SizeBox_Icon
   properties {bOverride_HeightOverride=true, bOverride_WidthOverride=true, heightOverride=36, widthOverride=36}
   slot {class=HorizontalBoxSlot, properties {horizontalAlignment=HAlign_Left, verticalAlignment=VAlign_Center, padding=<ltrb 16 0 0 0>}}
  }
  Switcher_Icon {
   widgetClassPath=&CommonVisibilitySwitcher
   bIsVariable=true
   parent=SizeBox_Icon
   widget=Switcher_Icon
   slot {class=SizeBoxSlot}
  }
  Loading_Throbber {
   widgetClassPath=/FortUILibrary/UIKit/Throbber/WBP_UIKit_Throbber
   bIsVariable=true
   parent=Switcher_Icon
   widget=Loading_Throbber
   slot {class=CommonVisibilitySwitcherSlot}
  }
  Error_Indicator {
   widgetClassPath=/FortUILibrary/UIKit/StatusIndicator/WBP_UIKit_StatusIndicator
   bIsVariable=true
   parent=Switcher_Icon
   widget=Error_Indicator
   properties {type=Negative, overrideStatusIndicatorVisibility=true}
   slot {class=CommonVisibilitySwitcherSlot}
  }
  Text {
   widgetClassPath=&CommonTextBlock
   bIsVariable=true
   parent=HorizontalBox_2
   widget=Text
   properties {
    autoWrapText=true
    bAutoCollapseWithEmptyText=true
    bIsScrollingEnabled=false
    mobileFontSizeMultiplier=1.2
    text=Loading 2D Models
    font {
     bForceMonospaced=false
     bMaterialIsStencil=false
     fontMaterial=null
     fontObject=&HeadingNow
     letterSpacing=0
     monospacedWidth=1
     size=15.75
     skewAmount=0
     typefaceFontName=74 Regular
     outlineSettings {
      bApplyOutlineToDropShadows=false
      bMiteredCorners=false
      bSeparateFillAlpha=false
      outlineMaterial=null
      outlineSize=0
      outlineColor=<rgba 0 0 0 1>
     }
    }
   }
   slot {class=HorizontalBoxSlot, properties {verticalAlignment=VAlign_Center, padding=<ltrb 16 8 16 8>, size {sizeRule=Fill, value=1}}}
  }
 ]
}
