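# Item preview screen widget that layers a camera frame, a 2D image preview, a download indicator and a loading-info panel.
# It tracks separate 2D and 3D load states plus an error message from download events and pushes them to the info widget, controlling its visibility.
# The info widget can be an externally assigned one, otherwise the built-in panel is used.
# Event dispatchers re-broadcast download, rotation, zoom, preview-action and next-item events to owners.
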WidgetBlueprint /FortniteCoreUI/Frontend/ItemPreview/WBP_ItemPreview
graphs [
 EventGraph {nodes [
   n1 = "On 2d Finished Loading (Widget_2dPreview)"!() -> (OutputDelegate, then, Success:bool, ImageUrl:String) {is=title K2Node_ComponentBoundEvent}
   > Set LoadState2D(LoadState2D:E_ItemPreview_LoadState Enum=Select(Option 0:E_ItemPreview_LoadState Enum=NewEnumerator1, Option 1:E_ItemPreview_LoadState Enum=NewEnumerator2, Index=@n1.Success) -> (ReturnValue:E_ItemPreview_LoadState Enum)) -> (Output_Get:E_ItemPreview_LoadState Enum=NewEnumerator3)
   > UpdatePreviewInfo()
   > Call On Item 2d Download Completed!(execute:Exec, self:WBP Item Preview Object Reference, ImageUrl:String=@n1.ImageUrl, Result:bool=@n1.Success) -> (then:Exec) {is=title K2Node_CallDelegate}
   n2 = "On 2d Started Loading (Widget_2dPreview)"!() -> (OutputDelegate, then, ImageUrl:String) {is=title K2Node_ComponentBoundEvent}
   > Call On Item 2d Download Started!(execute:Exec, self:WBP Item Preview Object Reference, ImageUrl:String=@n2.ImageUrl) -> (then:Exec) {is=title K2Node_CallDelegate}
   n3 = TriggerItemRotationChanged() {is=title K2Node_CustomEvent}
   > Call On Item Rotation Changed!(execute:Exec, self:WBP Item Preview Object Reference) -> (then:Exec) {is=title K2Node_CallDelegate}
   n4 = TriggerItemZoomLevelChanged() {is=title K2Node_CustomEvent}
   > Call On Item Zoom Level Changed!(execute:Exec, self:WBP Item Preview Object Reference) -> (then:Exec) {is=title K2Node_CallDelegate}
   Event Construct()
   n5 = Get ItemViewObject() -> (ItemViewObject:Fort Item View Object Reference)
   > Bind Event to On Item Rotation Changed(self:Fort Item View Object Reference=@n5, Delegate=@n3) {is=title K2Node_AddDelegate}
   > Bind Event to On Item Zoom Level Changed(self:Fort Item View Object Reference=@n5, Delegate=@n4) {is=title K2Node_AddDelegate}
   Event Destruct()
   n6 = Get ItemViewObject()
   > Unbind Event from On Item Rotation Changed(self:Fort Item View Object Reference=@n6, Delegate=@n3) {is=title K2Node_RemoveDelegate}
   > Unbind Event from On Item Zoom Level Changed(self:Fort Item View Object Reference=@n6, Delegate=@n4) {is=title K2Node_RemoveDelegate}
   n7 = Event OnPreviewActionsUpdated!() -> (OutputDelegate, then, PreviewActions:Array of Data Table Row Handle Structures) {is=title K2Node_Event}
   > Call On Preview Actions Updated Event!(execute:Exec, self:WBP Item Preview Object Reference, PreviewActions:Array of Data Table Row Handle Structures=@n7.PreviewActions) -> (then:Exec) {is=title K2Node_CallDelegate}
   n8 = Event OnPreviewActionButtonUpdated!() -> (OutputDelegate, then, RowDisplayName:"Text (by ref)", Button:Common Button Base Object Reference) {is=title K2Node_Event}
   > Call On Preview Action Button Updated Event!(
    execute:Exec
    self:WBP Item Preview Object Reference
    Row Display Name:Text=@n8.RowDisplayName
    Button:Common Button Base Object Reference=@n8.Button
   ) -> (then:Exec) {is=title K2Node_CallDelegate}
   Event OnItemPreviewShown!() -> (OutputDelegate, then, ItemDefinition:Fort Account Item Definition Object Reference) {is=title K2Node_Event}
   > On3dItemDownloaded(self:WBP Item Preview 2d Preview Object Reference=Get Widget_2dPreview() -> (Widget_2dPreview:WBP Item Preview 2d Preview Object Reference))
   Event On Deactivated()
   > DeactivateWidget(self=Get Widget_2dPreview())
   > ResetDownload()
   Event On Activated()
   n9 = Get Widget_2dPreview()
   > Set Is2dPreviewEnabled!(
    execute
    Is2dPreviewEnabled:bool=IsCosmetic2dPreviewEnabled!(self:Fort Item Preview Screen Object Reference) -> (ReturnValue:bool)
    self:WBP Item Preview 2d Preview Object Reference=@n9
   ) -> (then, Output_Get:bool)
   > ActivateWidget(self=@n9)
   > ResetDownload()
   n10 = Event PreConstruct()
   > n11 = Sequence()
   n12 = Select(Option 0=NewEnumerator3, Option 1=NewEnumerator0, Index=@n10.IsDesignTime)
   > Set LoadState2D(LoadState2D=@n12) -> (Output_Get=NewEnumerator3)
   > Set LoadState3D(LoadState3D:E_ItemPreview_LoadState Enum=@n12) -> (Output_Get:E_ItemPreview_LoadState Enum=NewEnumerator3)
   InitPreviewInfo(execute=@n11.then_1)
   > UpdatePreviewInfo()
   n13 = Event OnItemDownloadCompleted!() -> (OutputDelegate, then, Result:bool) {is=title K2Node_Event}
   > Set LoadState3D(LoadState3D=Select(Option 0=NewEnumerator1, Option 1=NewEnumerator2, Index=@n13.Result)) -> (Output_Get=NewEnumerator3)
   > UpdatePreviewInfo()
   > Call On Item 3d Download Completed!(
    execute:Exec
    self:WBP Item Preview Object Reference
    Item:Fort Item VM Object Reference=Get CurrentItemVM() -> (CurrentItemVM:Fort Item VM Object Reference)
    Result:bool=@n13.Result
   ) -> (then:Exec) {is=title K2Node_CallDelegate}
   Event OnItemDownloadStarted() {is=title K2Node_Event}
   > StartDownload()
   > Call On Item 3d Download Started!(execute:Exec, self:WBP Item Preview Object Reference, Item:Fort Item VM Object Reference=Get CurrentItemVM()) -> (then:Exec) {is=title K2Node_CallDelegate}
   n14 = Event OnItemDownloadErrored!() -> (OutputDelegate, then, ErrorCode:int=0, ModuleId:int=0, ModuleIdAndErrorCodeString:String, ErrorMessage:"Text (by ref)") {is=title K2Node_Event}
   > Set ErrorMessage(ErrorMessage:Text=@n14.ErrorMessage) -> (Output_Get:Text)
   > UpdatePreviewInfo()
   > Call On Item 3d Download Errored!(
    execute:Exec
    self:WBP Item Preview Object Reference
    Error Code:int=@n14.ErrorCode
    Module Id:int=@n14.ModuleId
    Module Id and Error Code String:String=@n14.ModuleIdAndErrorCodeString
    Error Message:Text=@n14.ErrorMessage
   ) -> (then:Exec) {is=title K2Node_CallDelegate}
   Event OnItemDownloadCancelled() {is=title K2Node_Event}
   > ResetDownload()
   > Call On Download Cancelled!(execute:Exec, self:WBP Item Preview Object Reference) -> (then:Exec) {is=title K2Node_CallDelegate}
   n15 = Event OnRequestNextCosmeticItem!() -> (
    OutputDelegate
    then
    TargetCosmeticQuery:"Gameplay Tag Query Structure (by ref)"
    ItemDefinition:Fort Account Item Definition Object Reference
   ) {is=title K2Node_Event}
   > Call On Next Item Requested!(
    execute:Exec
    self:WBP Item Preview Object Reference
    Target Cosmetic Query:Gameplay Tag Query Structure=@n15.TargetCosmeticQuery
    Item Definition:Fort Account Item Definition Object Reference=@n15.ItemDefinition
   ) -> (then:Exec) {is=title K2Node_CallDelegate}
   ClearPreviewItem(self:Fort Item Preview Widget Object Reference)
  ]}
 InitPreviewInfo {nodes [
   InitPreviewInfo() {is=title K2Node_FunctionEntry}
   n1 = Get InfoWidget() -> (InfoWidget:WBP Item Preview Info Object Reference)
   n2 = Get Widget_PreviewInfo_Internal() -> (Widget_PreviewInfo_Internal:WBP Item Preview Info Object Reference)
   > n3 = Branch(Condition=AND Boolean(
     A=IsValid(Object=@n1)
     B="Not Equal (Object)"(A:Object Reference=@n1, B:Object Reference=@n2) -> (ReturnValue:bool=~) {is=title K2Node_PromotableOperator}
    ))
   n4 = Get Widget_PreviewInfo_Internal()
   > n5 = RemoveFromParent(self=@n4)
   Set Visibility(
    execute=[@n8.then, @n5.then]
    Visibility=Select(
     Option 0:ESlateVisibility Enum=Collapsed
     Option 1:ESlateVisibility Enum=HitTestInvisible
     Index=ShouldShowItemPreviewInfo!() -> (IsShown:bool)
    ) -> (ReturnValue:ESlateVisibility Enum)
    self=Get InfoWidget()
   )
   n6 = Set InfoWidget(execute=@n3.else, InfoWidget:WBP Item Preview Info Object Reference=@n4) -> (Output_Get:WBP Item Preview Info Object Reference)
   n7 = SlotAsOverlaySlot(Widget=@n6)
   > SetHorizontalAlignment(self=@n7, InHorizontalAlignment=Get InfoAlignment_H() -> (InfoAlignment_H:EHorizontalAlignment Enum=HAlign_Fill))
   > SetVerticalAlignment(self=@n7, InVerticalAlignment=Get InfoAlignment_V() -> (InfoAlignment_V:EVerticalAlignment Enum=VAlign_Fill))
   > n8 = SetPadding!(execute, self:Overlay Slot Object Reference=@n7, InPadding:Margin Structure=Get InfoPadding() -> (InfoPadding:Margin Structure)) -> (then)
  ]}
 OnDownloadCancelled {nodes [OnDownloadCancelled() {is=title K2Node_FunctionEntry}]}
 OnItem2dDownloadCompleted {nodes [OnItem2dDownloadCompleted!() -> (then, ImageUrl:String, Result:bool=~) {is=title K2Node_FunctionEntry}]}
 OnItem2dDownloadStarted {nodes [OnItem2dDownloadStarted!() -> (then, ImageUrl:String) {is=title K2Node_FunctionEntry}]}
 OnItem3dDownloadCompleted {nodes [OnItem3dDownloadCompleted!() -> (then, Item:Fort Item VM Object Reference, Result:bool=~) {is=title K2Node_FunctionEntry}]}
 OnItem3dDownloadErrored {nodes [OnItem3dDownloadErrored!() -> (then, Error Code:int, Module Id:int, Module Id and Error Code String:String, Error Message:Text) {is=title K2Node_FunctionEntry}]}
 OnItem3dDownloadStarted {nodes [OnItem3dDownloadStarted!() -> (then, Item:Fort Item VM Object Reference) {is=title K2Node_FunctionEntry}]}
 OnItemRotationChanged {nodes [OnItemRotationChanged() {is=title K2Node_FunctionEntry}]}
 OnItemZoomLevelChanged {nodes [OnItemZoomLevelChanged() {is=title K2Node_FunctionEntry}]}
 OnNextItemRequested {nodes [OnNextItemRequested!() -> (then, Target Cosmetic Query:Gameplay Tag Query Structure, Item Definition:Fort Account Item Definition Object Reference) {is=title K2Node_FunctionEntry}]}
 OnPreviewActionButtonUpdatedEvent {nodes [OnPreviewActionButtonUpdatedEvent!() -> (then, Row Display Name:Text, Button:Common Button Base Object Reference) {is=title K2Node_FunctionEntry}]}
 OnPreviewActionsUpdatedEvent {nodes [OnPreviewActionsUpdatedEvent!() -> (then, PreviewActions:Array of Data Table Row Handle Structures) {is=title K2Node_FunctionEntry}]}
 ResetDownload {nodes [
   ResetDownload() {is=title K2Node_FunctionEntry}
   > Set LoadState2D(LoadState2D=NewEnumerator3) -> (Output_Get=NewEnumerator3)
   > Set LoadState3D(LoadState3D=NewEnumerator3) -> (Output_Get=NewEnumerator3)
   > Set ErrorMessage()
   > UpdatePreviewInfo()
  ]}
 ShouldShowItemPreviewInfo {nodes [
   ShouldShowItemPreviewInfo() {is=title K2Node_FunctionEntry}
   n1 = Get InfoWidget()
   > Return Node!(
    execute
    IsShown:bool=AND Boolean!(
     A=HasLoadingInfoToShow!(self:WBP Item Preview Info Object Reference=@n1) -> (HasInfo:bool)
     B=Get ShowPreviewInfo() -> (ShowPreviewInfo:bool)
     C:bool=IsItemPreviewDownloadInfoEnabled!(self:Fort Item Preview Screen Object Reference) -> (ReturnValue:bool)
    ) -> (ReturnValue)
   )
  ]}
 StartDownload {nodes [
   StartDownload() {is=title K2Node_FunctionEntry}
   > Set LoadState2D(LoadState2D=Select(
     Option 0:E_ItemPreview_LoadState Enum=NewEnumerator2
     Option 1:E_ItemPreview_LoadState Enum=NewEnumerator0
     Index=IsCosmetic2dPreviewEnabled()
    ) -> (ReturnValue:E_ItemPreview_LoadState Enum)) -> (Output_Get=NewEnumerator3)
   > Set LoadState3D(LoadState3D=NewEnumerator0) -> (Output_Get=NewEnumerator3)
   > Set ErrorMessage()
   > UpdatePreviewInfo()
  ]}
 UpdatePreviewInfo {nodes [
   UpdatePreviewInfo() {is=title K2Node_FunctionEntry}
   > UpdateLoading!(
    execute
    self:WBP Item Preview Info Object Reference=Get InfoWidget()
    LoadState2D:E_ItemPreview_LoadState Enum=Get LoadState2D() -> (LoadState2D:E_ItemPreview_LoadState Enum=NewEnumerator3)
    LoadState3D:E_ItemPreview_LoadState Enum=Get LoadState3D() -> (LoadState3D:E_ItemPreview_LoadState Enum=NewEnumerator3)
    ErrorMessage:Text=Get ErrorMessage() -> (ErrorMessage:Text)
   ) -> (then)
   > Set Visibility(
    Visibility=Select(
     Option 0:ESlateVisibility Enum=Collapsed
     Option 1:ESlateVisibility Enum=SelfHitTestInvisible
     Index=ShouldShowItemPreviewInfo() {is=title K2Node_CallFunction}
    ) -> (ReturnValue:ESlateVisibility Enum)
    self=Get InfoWidget()
   )
  ]}
]
variables [
 ErrorMessage {type=string, default=""}
 InfoAlignment_H {type=EHorizontalAlignment, default=HAlign_Right}
 InfoAlignment_V {type=EVerticalAlignment, default=VAlign_Top}
 InfoPadding {type=Margin, default=<ltrb 60 60 60 60>}
 InfoWidget {type=WBP_ItemPreview_Info_C, default=null}
 LoadState2D {type=E_ItemPreview_LoadState, default=Unset}
 LoadState3D {type=E_ItemPreview_LoadState, default=Unset}
 OnDownloadCancelled {}
 OnItem2dDownloadCompleted {}
 OnItem2dDownloadStarted {}
 OnItem3dDownloadCompleted {}
 OnItem3dDownloadErrored {}
 OnItem3dDownloadStarted {}
 OnItemRotationChanged {}
 OnItemZoomLevelChanged {}
 OnNextItemRequested {}
 OnPreviewActionButtonUpdatedEvent {}
 OnPreviewActionsUpdatedEvent {}
 ShowPreviewInfo {type=boolean, default=true}
]
widgets {
 info {inheritedWidgetCount=6, parentClass=/Script/FortniteUI.FortItemPreviewWidget, rootWidgetClass=&Overlay, widgetCount=10}
 widgets [
  Overlay_0 {widgetClassPath=&Overlay, widget=Overlay_0}
  Widget_CameraFrame {
   widgetClassPath=&SizeBox
   bInherited=true
   bIsVariable=true
   parent=Overlay_0
   widget=Widget_CameraFrame
   properties {heightOverride=1920}
   slot {class=OverlaySlot, properties {horizontalAlignment=HAlign_Fill, verticalAlignment=VAlign_Fill}}
  }
  Widget_2dPreview {
   widgetClassPath=&/WBP_ItemPreview_2dPreview
   bIsVariable=true
   parent=Overlay_0
   widget=Widget_2dPreview
   properties {bAutoActivate=false}
   slot {class=OverlaySlot, properties {horizontalAlignment=HAlign_Fill, verticalAlignment=VAlign_Fill}}
  }
  ItemDownloadIndicator {
   widgetClassPath=/Game/UI/ContentOnDemand/ItemDownloadIndicator/ItemDownloadIndicator
   bInherited=true
   bIsVariable=true
   parent=Overlay_0
   widget=ItemDownloadIndicator
   slot {class=OverlaySlot, properties {horizontalAlignment=HAlign_Fill, verticalAlignment=VAlign_Fill}}
  }
  Widget_PreviewInfo_Internal {
   widgetClassPath=&/WBP_ItemPreview_Info
   bIsVariable=true
   parent=Overlay_0
   widget=Widget_PreviewInfo_Internal
   slot {class=OverlaySlot, properties {horizontalAlignment=HAlign_Right, padding=<ltrb 0 20 20 0>}}
  }
  HorizontalBox_0 {
   widgetClassPath=&HorizontalBox
   parent=Overlay_0
   widget=HorizontalBox_0
   slot {class=OverlaySlot, properties {horizontalAlignment=HAlign_Right, verticalAlignment=VAlign_Bottom}}
  }
  Text_DebugPreviewDetails {
   widgetClassPath=&CommonTextBlock
   bInherited=true
   parent=HorizontalBox_0
   widget=Text_DebugPreviewDetails
   properties {
    wrapTextAt=650
    font {
     bForceMonospaced=false
     bMaterialIsStencil=false
     fontMaterial=null
     fontObject=&NotoSans
     letterSpacing=0
     monospacedWidth=1
     size=12
     skewAmount=0
     typefaceFontName=NotoSans-Regular
     outlineSettings {
      bApplyOutlineToDropShadows=false
      bMiteredCorners=false
      bSeparateFillAlpha=false
      outlineMaterial=null
      outlineSize=1
      outlineColor=<rgba 0 0 0 1>
     }
    }
    shadowOffset=<4 0>
   }
   slot {class=HorizontalBoxSlot, properties {horizontalAlignment=HAlign_Center, verticalAlignment=VAlign_Center}}
  }
  Button_PreviewAction {widgetClassPath=&CommonButtonBase, bInherited=true, widget=null}
  Button_PreviewAction1 {widgetClassPath=&CommonButtonBase, bInherited=true, widget=null}
  Button_Camera {widgetClassPath=&CommonButtonBase, bInherited=true, widget=null}
 ]
}
