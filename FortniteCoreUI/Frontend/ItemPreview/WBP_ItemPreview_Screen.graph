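# Activatable item preview screen that hosts WBP_ItemPreview over a vault-world background, with description, cost, wallet text, interact button and action bar.
# Activation starts the preview and sets the background; Back clears it and deactivates both preview and screen.
# Previous/Next clicks call empty custom events, while UpdateArrowButtonVisibility collapses arrows at list ends.
# Named slots and a decorator-setting function allow customization, and design-time PreConstruct fills placeholder text.
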WidgetBlueprint /FortniteCoreUI/Frontend/ItemPreview/WBP_ItemPreview_Screen
graphs [
 EventGraph {nodes [
   "On Clicked (Button_Back)"!() -> (OutputDelegate, then, Button:Common Button Base Object Reference) {is=title K2Node_ComponentBoundEvent}
   > UpdateBackground!(
    execute
    self:Fort Item Preview Background Object Reference=Get FortItemPreviewBackground() -> (FortItemPreviewBackground:Fort Item Preview Background Object Reference)
    LoadedBackground:Object Reference
    TransitionDirection:EVaultWorldTransitionDirection Enum=None
    PrimaryColor:Linear Color Structure=(R=0,G=0,B=0,A=0)
    SecondaryColor:Linear Color Structure=(R=0,G=0,B=0,A=0)
   ) -> (then)
   > DeactivateWidget(self=Get WBP_ItemPreview() -> (WBP_ItemPreview:WBP Item Preview Object Reference))
   > DeactivateWidget()
   "On Clicked (Button_Previous)"!() -> (OutputDelegate, then, Button:Common Button Base Object Reference) {is=title K2Node_ComponentBoundEvent}
   > Previous Button Clicked()
   "On Clicked (Button_Next)"!() -> (OutputDelegate, then, Button:Common Button Base Object Reference) {is=title K2Node_ComponentBoundEvent}
   > Next Button Clicked()
   n1 = Event PreConstruct()
   > Branch(Condition=@n1.IsDesignTime)
   > UpdateForPlaceholder()
   Event On Activated()
   > ActivateWidget(self=Get WBP_ItemPreview())
   > UpdateBackground(
    self=Get FortItemPreviewBackground()
    LoadedBackground=Get Vaultworld Background() -> (Vaultworld Background:Object Reference)
    TransitionDirection=None
    PrimaryColor=(R=0,G=0,B=0,A=0)
    SecondaryColor=(R=0,G=0,B=0,A=0)
   )
   Previous Button Clicked() {is=title K2Node_CustomEvent}
   Next Button Clicked() {is=title K2Node_CustomEvent}
  ]}
 UpdateArrowButtonVisibility {nodes [
   UpdateArrowButtonVisibility!() -> (then, CurrentIndex:int, LastIndex:int) {is=title K2Node_FunctionEntry}
   n1 = Get CurrentIndex() -> (CurrentIndex:int=0)
   > SetVisibility(
    self=Get Button_Previous() -> (Button_Previous:WBP UIKit Button Generic Object Reference)
    InVisibility=Select(Option 0=SelfHitTestInvisible, Option 1=Collapsed, Index="Equal (Integer)"(B=@n1))
   )
   n2 = Get LastIndex() -> (LastIndex:int=0)
   > SetVisibility(
    self=Get Button_Next() -> (Button_Next:WBP UIKit Button Generic Object Reference)
    InVisibility=Select(Option 0=SelfHitTestInvisible, Option 1=Collapsed, Index="Equal (Integer)"(A=@n1, B=@n2))
   )
  ]}
 UpdateForCustomVisuals {nodes [
   n1 = UpdateForCustomVisuals!() -> (then, InDecoratorClasses:Array of Rich Text Block Decorator Class References) {is=title K2Node_FunctionEntry}
   > SetDecorators!(
    execute
    self:Rich Text Block Object Reference=[
     Get RichText_CurrenyCurrencyAmount() -> (RichText_CurrenyCurrencyAmount:Common Rich Text Block Object Reference)
     Get RichText_ItemCost() -> (RichText_ItemCost:Common Rich Text Block Object Reference)
    ]
    InDecoratorClasses:Array of Rich Text Block Decorator Class References=@n1
   ) -> (then)
  ]}
 UpdateForPlaceholder {nodes [
   UpdateForPlaceholder() {is=title K2Node_FunctionEntry}
   > SetText(self=Get RichText_ItemCost(), InText="INVTEXT(\"[PH] Item Cost\")")
   > SetText(self=Get RichText_CurrenyCurrencyAmount(), InText="INVTEXT(\"[PH] Current Wallet Amount\")")
   > "SetText (Text)"(self=Get Text_ItemInfo() -> (Text_ItemInfo:Common Text Object Reference), InText="INVTEXT(\"[PH] Item Info\")")
   n1 = Get Button_Interact() -> (Button_Interact:WBP UIKit Button Loud Object Reference)
   > SetVisibility(self=@n1, InVisibility=SelfHitTestInvisible)
   > Set Text(Text="INVTEXT(\"[PH] Interact\")", self:Fort CTAButton Object Reference=@n1)
  ]}
]
variables [Vaultworld Background {type=Object, default=/Game/UI/Foundation/Environment/Shell/MaterialInstances/MI_UI_VaultWorld_Background_Generic}]
widgets {
 info {
  inheritedWidgetCount=0
  namedSlotCount=3
  parentClass=&CommonActivatableWidget
  rootWidgetClass=&Overlay
  widgetCount=27
 }
 widgets [
  Overlay_69 {widgetClassPath=&Overlay, widget=Overlay_69}
  WBP_CaptureForPostBufferUpdate {
   widgetClassPath=&WBP_CaptureForPostBufferUpdate
   bIsVariable=true
   parent=Overlay_69
   slot {class=OverlaySlot, properties {horizontalAlignment=HAlign_Fill, verticalAlignment=VAlign_Fill}}
  }
  FortItemPreviewBackground {
   widgetClassPath=/Script/FortniteUI.FortItemPreviewBackground
   bIsVariable=true
   parent=Overlay_69
   widget=FortItemPreviewBackground
   properties {previewSceneChangerTag {tagName=UI.ItemPreview.SceneChanger.MPItemShop}}
   slot {class=OverlaySlot, properties {horizontalAlignment=HAlign_Fill, verticalAlignment=VAlign_Fill}}
  }
  Slot_CustomBackground {
   widgetClassPath=&NamedSlot
   bIsVariable=true
   parent=Overlay_69
   widget=Slot_CustomBackground
   slot {class=OverlaySlot, properties {horizontalAlignment=HAlign_Fill, verticalAlignment=VAlign_Fill}}
  }
  SafeZone {
   widgetClassPath=&SafeZone
   bIsVariable=true
   parent=Overlay_69
   widget=SafeZone
   slot {class=OverlaySlot, properties {horizontalAlignment=HAlign_Fill, verticalAlignment=VAlign_Fill}}
  }
  SizeBox_2 {
   widgetClassPath=&SizeBox
   parent=SafeZone
   widget=SizeBox_2
   properties {maxAspectRatio=3.9, minAspectRatio=1.7778}
   slot {class=SafeZoneSlot}
  }
  GridPanel_0 {
   widgetClassPath=&GridPanel
   parent=SizeBox_2
   widget=GridPanel_0
   properties {columnFill [1,1,0.8], rowFill [0,1,0]}
   slot {class=SizeBoxSlot, properties {padding=<ltrb 80 40 80 40>}}
  }
  WBP_ItemPreview {
   widgetClassPath=&/WBP_ItemPreview
   bIsVariable=true
   parent=GridPanel_0
   widget=WBP_ItemPreview
   properties {
    itemViewObject=&self:WidgetTree.WBP_ItemPreview.ItemViewObject
    defaultPreviewContextTags {gameplayTags [{tagName=UI.ItemPreview.PreviewType.OnPawn}]}
    previewActionButtons [null,null,null]
    previewSceneChangerTag {tagName=UI.ItemPreview.SceneChanger.MPItemShop}
   }
   slot {class=GridSlot, properties {columnSpan=3, rowSpan=2, padding=<ltrb 128 0 128 0>}}
  }
  RichText_CurrenyCurrencyAmount {
   widgetClassPath=&CommonRichTextBlock
   bIsVariable=true
   parent=GridPanel_0
   widget=RichText_CurrenyCurrencyAmount
   properties {
    bOverrideDefaultStyle=true
    justification=Right
    lineHeightPercentage=1.15
    mobileTextBlockScale=1.2
    textStyleSet=/Game/Athena/Items/Cosmetics/Metadata/ItemDescriptionTextStyle
    wrapTextAt=400
    defaultTextStyleOverride {
     overflowPolicy=Clip
     transformPolicy=None
     colorAndOpacity {colorUseRule=UseColor_Specified, specifiedColor=<rgba 1 1 1 0.7>}
     font {
      bForceMonospaced=false
      bMaterialIsStencil=false
      fontMaterial=null
      fontObject=&HeadingNow
      letterSpacing=0
      monospacedWidth=1
      size=15.75
      skewAmount=0
      typefaceFontName=74 Regular
      outlineSettings {
       bApplyOutlineToDropShadows=true
       bMiteredCorners=false
       bSeparateFillAlpha=false
       outlineMaterial=null
       outlineSize=0
       outlineColor=<rgba 0 0 0 1>
      }
     }
     highlightColor {colorUseRule=UseColor_Specified, specifiedColor=<rgba 0 0 0 0>}
     highlightShape {
      bIsDynamicallyLoaded=false
      drawAs=Image
      imageType=NoImage
      mirroring=NoMirror
      resourceName=None
      resourceObject=null
      tiling=NoTile
      imageSize=<32 32>
      margin=<ltrb 0 0 0 0>
      outlineSettings {
       bUseBrushTransparency=false
       roundingType=HalfHeightRadius
       width=0
       color {colorUseRule=UseColor_Specified, specifiedColor=<rgba 0 0 0 0>}
       cornerRadii=<0 0 0 0>
      }
      tintColor {colorUseRule=UseColor_Specified, specifiedColor=<rgba 1 1 1 1>}
      uVRegion {bIsValid=false, max=<0 0>, min=<0 0>}
     }
     selectedBackgroundColor {colorUseRule=UseColor_Foreground, specifiedColor=<rgba 1 0 1 1>}
     shadowColorAndOpacity=<rgba 0 0 0 1>
     shadowOffset=<0 0>
     strikeBrush {
      bIsDynamicallyLoaded=false
      drawAs=Image
      imageType=NoImage
      mirroring=NoMirror
      resourceName=None
      resourceObject=null
      tiling=NoTile
      imageSize=<32 32>
      margin=<ltrb 0 0 0 0>
      outlineSettings {
       bUseBrushTransparency=false
       roundingType=HalfHeightRadius
       width=0
       color {colorUseRule=UseColor_Specified, specifiedColor=<rgba 0 0 0 0>}
       cornerRadii=<0 0 0 0>
      }
      tintColor {colorUseRule=UseColor_Specified, specifiedColor=<rgba 1 1 1 1>}
      uVRegion {bIsValid=false, max=<0 0>, min=<0 0>}
     }
     underlineBrush {
      bIsDynamicallyLoaded=false
      drawAs=Image
      imageType=NoImage
      mirroring=NoMirror
      resourceName=None
      resourceObject=null
      tiling=NoTile
      imageSize=<32 32>
      margin=<ltrb 0 0 0 0>
      outlineSettings {
       bUseBrushTransparency=false
       roundingType=HalfHeightRadius
       width=0
       color {colorUseRule=UseColor_Specified, specifiedColor=<rgba 0 0 0 0>}
       cornerRadii=<0 0 0 0>
      }
      tintColor {colorUseRule=UseColor_Specified, specifiedColor=<rgba 1 1 1 1>}
      uVRegion {bIsValid=false, max=<0 0>, min=<0 0>}
     }
    }
   }
   slot {class=GridSlot, properties {column=2, verticalAlignment=VAlign_Top, padding=<ltrb 0 64 64 0>}}
  }
  NavButtons_Overlay {
   widgetClassPath=&Overlay
   parent=GridPanel_0
   widget=NavButtons_Overlay
   slot {class=GridSlot, properties {columnSpan=3, rowSpan=3}}
  }
  Button_Previous {
   widgetClassPath=&WBP_UIKit_Button_Generic
   bIsVariable=true
   parent=NavButtons_Overlay
   widget=Button_Previous
   properties {
    bHideInputActionWithKeyboard=true
    bIsFocusable=false
    backgroundMaterial=/FortUILibrary/UIKit/Buttons/MaterialInstances/CTAButtons/MI_UIKit_BgRegular
    enableFocusedSelectedState=false
    minHeight=64
    minWidth=64
    outlineMaterial=/FortUILibrary/UIKit/Buttons/MaterialInstances/CTAButtons/MI_UIKit_ButtonOutline
    priority=100
    promptHorizontalAlignment=HAlign_Center
    promptVerticalAlignment=VAlign_Bottom
    useBackgroundBlock=false
    useTextBlock=false
    buttonPadding=<ltrb 0 0 0 0>
    iconSize=<64 64>
    promptNudge=<0 64>
    secondTextPadding=<ltrb 0 0 0 0>
    statusIndicatorPadding=<ltrb 0 0 0 0>
    triggeringInputAction {dataTable=&InputActionDataTable, rowName=GenericLeftShoulder}
   }
   slot {class=OverlaySlot, properties {verticalAlignment=VAlign_Center}}
  }
  Slot_PreviousIcon {widgetClassPath=&NamedSlot, bIsVariable=true, namedSlotHost=Button_Previous, widget=Slot_PreviousIcon}
  Button_Next {
   widgetClassPath=&WBP_UIKit_Button_Generic
   bIsVariable=true
   parent=NavButtons_Overlay
   widget=Button_Next
   properties {
    bHideInputActionWithKeyboard=true
    bIsFocusable=false
    bShouldUseFallbackDefaultInputAction=false
    backgroundMaterial=&MI_UIKit_BgRegular
    enableFocusedSelectedState=false
    minHeight=64
    minWidth=64
    outlineMaterial=&MI_UIKit_ButtonOutline
    priority=100
    promptHorizontalAlignment=HAlign_Center
    promptVerticalAlignment=VAlign_Bottom
    useBackgroundBlock=false
    useTextBlock=false
    buttonPadding=<ltrb 0 0 0 0>
    iconSize=<64 64>
    promptNudge=<0 64>
    triggeringInputAction {dataTable=&InputActionDataTable, rowName=GenericRightShoulder}
   }
   slot {class=OverlaySlot, properties {horizontalAlignment=HAlign_Right, verticalAlignment=VAlign_Center}}
  }
  Slot_NextIcon {widgetClassPath=&NamedSlot, bIsVariable=true, namedSlotHost=Button_Next, widget=Slot_NextIcon}
  ItemDescription {
   widgetClassPath=/FortUILibrary/UIKit/Items/WBP_UIKit_ItemDescription_Compact
   bIsVariable=true
   parent=GridPanel_0
   widget=ItemDescription
   properties {titleMaxHeight=250, input Action Row {dataTable=&InputActionDataTable, rowName=GenericFaceButtonTop_Hold}}
   slot {class=GridSlot, properties {padding=<ltrb 64 64 0 0>}}
  }
  Info_VBox {
   widgetClassPath=&VerticalBox
   bIsVariable=true
   parent=GridPanel_0
   widget=Info_VBox
   slot {class=GridSlot, properties {columnSpan=3, row=2, verticalAlignment=VAlign_Bottom, padding=<ltrb 0 0 0 16>}}
  }
  Slot_CenterSection {
   widgetClassPath=&NamedSlot
   bIsVariable=true
   parent=Info_VBox
   widget=Slot_CenterSection
   slot {class=VerticalBoxSlot, properties {horizontalAlignment=HAlign_Center}}
  }
  RichText_ItemCost {
   widgetClassPath=&CommonRichTextBlock
   bIsVariable=true
   parent=Info_VBox
   widget=RichText_ItemCost
   properties {
    autoWrapText=true
    bOverrideDefaultStyle=true
    justification=Center
    lineHeightPercentage=1.15
    mobileTextBlockScale=1.2
    textStyleSet=&ItemDescriptionTextStyle
    wrapTextAt=700
    defaultTextStyleOverride {
     overflowPolicy=Clip
     transformPolicy=None
     colorAndOpacity {colorUseRule=UseColor_Specified, specifiedColor=<rgba 1 1 1 1>}
     font {
      bForceMonospaced=false
      bMaterialIsStencil=false
      fontMaterial=null
      fontObject=&HeadingNow
      letterSpacing=0
      monospacedWidth=1
      size=18
      skewAmount=0
      typefaceFontName=74 Regular
      outlineSettings {
       bApplyOutlineToDropShadows=true
       bMiteredCorners=false
       bSeparateFillAlpha=false
       outlineMaterial=null
       outlineSize=0
       outlineColor=<rgba 0 0 0 1>
      }
     }
     highlightColor {colorUseRule=UseColor_Specified, specifiedColor=<rgba 0 0 0 0>}
     highlightShape {
      bIsDynamicallyLoaded=false
      drawAs=Image
      imageType=NoImage
      mirroring=NoMirror
      resourceName=None
      resourceObject=null
      tiling=NoTile
      imageSize=<32 32>
      margin=<ltrb 0 0 0 0>
      outlineSettings {
       bUseBrushTransparency=false
       roundingType=HalfHeightRadius
       width=0
       color {colorUseRule=UseColor_Specified, specifiedColor=<rgba 0 0 0 0>}
       cornerRadii=<0 0 0 0>
      }
      tintColor {colorUseRule=UseColor_Specified, specifiedColor=<rgba 1 1 1 1>}
      uVRegion {bIsValid=false, max=<0 0>, min=<0 0>}
     }
     selectedBackgroundColor {colorUseRule=UseColor_Foreground, specifiedColor=<rgba 1 0 1 1>}
     shadowColorAndOpacity=<rgba 0 0 0 1>
     shadowOffset=<0 0>
     strikeBrush {
      bIsDynamicallyLoaded=false
      drawAs=Image
      imageType=NoImage
      mirroring=NoMirror
      resourceName=None
      resourceObject=null
      tiling=NoTile
      imageSize=<32 32>
      margin=<ltrb 0 0 0 0>
      outlineSettings {
       bUseBrushTransparency=false
       roundingType=HalfHeightRadius
       width=0
       color {colorUseRule=UseColor_Specified, specifiedColor=<rgba 0 0 0 0>}
       cornerRadii=<0 0 0 0>
      }
      tintColor {colorUseRule=UseColor_Specified, specifiedColor=<rgba 1 1 1 1>}
      uVRegion {bIsValid=false, max=<0 0>, min=<0 0>}
     }
     underlineBrush {
      bIsDynamicallyLoaded=false
      drawAs=Image
      imageType=NoImage
      mirroring=NoMirror
      resourceName=None
      resourceObject=null
      tiling=NoTile
      imageSize=<32 32>
      margin=<ltrb 0 0 0 0>
      outlineSettings {
       bUseBrushTransparency=false
       roundingType=HalfHeightRadius
       width=0
       color {colorUseRule=UseColor_Specified, specifiedColor=<rgba 0 0 0 0>}
       cornerRadii=<0 0 0 0>
      }
      tintColor {colorUseRule=UseColor_Specified, specifiedColor=<rgba 1 1 1 1>}
      uVRegion {bIsValid=false, max=<0 0>, min=<0 0>}
     }
    }
   }
   slot {class=VerticalBoxSlot, properties {verticalAlignment=VAlign_Top, padding=<ltrb 0 4 0 4>}}
  }
  Text_ItemInfo {
   widgetClassPath=&CommonTextBlock
   bIsVariable=true
   parent=Info_VBox
   widget=Text_ItemInfo
   properties {
    justification=Center
    colorAndOpacity {colorUseRule=UseColor_Specified, specifiedColor=<rgba 0.0437 1 0.2307 1>}
    font {
     bForceMonospaced=false
     bMaterialIsStencil=false
     fontMaterial=null
     fontObject=&HeadingNow
     letterSpacing=0
     monospacedWidth=1
     size=19.5
     skewAmount=0
     typefaceFontName=75 Medium
     outlineSettings {
      bApplyOutlineToDropShadows=true
      bMiteredCorners=false
      bSeparateFillAlpha=false
      outlineMaterial=null
      outlineSize=0
      outlineColor=<rgba 0 0 0 1>
     }
    }
    shadowOffset=<0 0>
   }
   slot {class=VerticalBoxSlot}
  }
  Button_Interact {
   widgetClassPath=/FortUILibrary/UIKit/Buttons/WBP_UIKit_Button_Loud
   bIsVariable=true
   parent=Info_VBox
   widget=Button_Interact
   properties {text="", visibility=Collapsed, triggeringInputAction {dataTable=&InputActionDataTable, rowName=GenericFaceButtonLeft}}
   slot {class=VerticalBoxSlot, properties {horizontalAlignment=HAlign_Center, padding=<ltrb 0 8 0 0>}}
  }
  WBP_ActionBarContainer {
   widgetClassPath=/Game/UI/Frontend/BottomBar/WBP_ActionBarContainer
   bIsVariable=true
   parent=GridPanel_0
   widget=WBP_ActionBarContainer
   slot {class=GridSlot, properties {columnSpan=3, horizontalAlignment=HAlign_Right, row=3}}
  }
  Button_Bar {widgetClassPath=&HorizontalBox, namedSlotHost=WBP_ActionBarContainer, widget=Button_Bar}
  ActionButtonsWrapper {widgetClassPath=&HorizontalBox, bIsVariable=true, parent=Button_Bar, widget=ActionButtonsWrapper, slot {class=HorizontalBoxSlot}}
  Button_PreviewAction {
   widgetClassPath=/FortUILibrary/UIKit/Buttons/WBP_UIKit_Button_Quiet
   parent=ActionButtonsWrapper
   widget=Button_PreviewAction
   properties {
    bAutoUpdateFromInputAction=true
    bDisplayInputActionWhenNotInteractable=false
    bIsFocusable=false
    bShouldUseFallbackDefaultInputAction=false
    minWidth=48
    text=""
   }
   slot {class=HorizontalBoxSlot, properties {padding=<ltrb 0 0 24 0>}}
  }
  Button_PreviewAction1 {
   widgetClassPath=&WBP_UIKit_Button_Quiet
   parent=ActionButtonsWrapper
   widget=Button_PreviewAction1
   properties {
    bAutoUpdateFromInputAction=true
    bDisplayInputActionWhenNotInteractable=false
    bIsFocusable=false
    bShouldUseFallbackDefaultInputAction=false
    minWidth=48
    text=""
   }
   slot {class=HorizontalBoxSlot, properties {padding=<ltrb 0 0 24 0>}}
  }
  Button_PreviewAction2 {
   widgetClassPath=&WBP_UIKit_Button_Quiet
   parent=ActionButtonsWrapper
   widget=Button_PreviewAction2
   properties {
    bAutoUpdateFromInputAction=true
    bDisplayInputActionWhenNotInteractable=false
    bIsFocusable=false
    bShouldUseFallbackDefaultInputAction=false
    minWidth=48
    text=""
   }
   slot {class=HorizontalBoxSlot, properties {padding=<ltrb 0 0 24 0>}}
  }
  Button_Back {
   widgetClassPath=&WBP_UIKit_Button_Quiet
   bIsVariable=true
   parent=Button_Bar
   widget=Button_Back
   properties {
    bIsFocusable=false
    text=Back
    visibility=Visible
    triggeringInputAction {dataTable=&InputActionDataTable, rowName=GenericBack}
   }
   slot {class=HorizontalBoxSlot, properties {verticalAlignment=VAlign_Center}}
  }
 ]
}
